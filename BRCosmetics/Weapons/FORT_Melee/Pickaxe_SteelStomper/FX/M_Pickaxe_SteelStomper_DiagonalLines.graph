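# Translucent unlit material for Niagara ribbon effects on the Steel Stomper pickaxe, drawing orange and dark-grey diagonal stripes.
# Stripes come from a sine over fixed-angle rotated UV1, sharpened by smoothstep, then tinted by particle color and passed through a shared emissive-adjustment function.
# Opacity multiplies particle alpha by procedural UV0 fades along the length and across both edges, broken up by packed noise and softened by depth fade.
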Material /BRCosmetics/Weapons/FORT_Melee/Pickaxe_SteelStomper/FX/M_Pickaxe_SteelStomper_DiagonalLines
graph {
 properties {
  bAutomaticallySetUsageInEditor=false
  bUsedWithNiagaraRibbons=true
  blendMode=BLEND_Translucent
  pixelDepthOffsetMode=PDOM_Legacy
  shadingModel=MSM_Unlit
  usedShadingModels=Unlit
 }
 nodes [
  n1 = ParticleColor()
  n2 = TextureCoordinate()
  n3 = ComponentMask(None=@n2) {r=true}
  n4 = Multiply(A=ComponentMask(None=@n2) {g=true}) {constB=0.5}
  n5 = Multiply(A=SmoothStep(Value=OneMinus(None=@n3)) {constMin=0.1}, B=SmoothStep(Value=Sine(None=@n4)) {constMin=0.2})
  n6 = TextureCoordinate()
  n7 = Constant2Vector() {g=1, r=2}
  n8 = LinearInterpolate(Alpha=TextureSample(UVs=Multiply(A=@n6, B=@n7)) {samplerType=SAMPLERTYPE_Masks, texture=&/M_Pickaxe_SteelStomper_PackedNoise}.R) {constA=0.4, constB=2}
  DepthFade(Opacity=Multiply(A=@n1.A, B=Multiply(A=@n5, B=@n8)), FadeDistance=Constant() {r=15})
  n9 = Constant3Vector() {constant=<rgba 1.2 0.1412 0 1>}
  n10 = Constant3Vector() {constant=<rgba 0.0417 0.0417 0.0417 0>}
  n11 = ComponentMask(None=Rotator(Coordinate=TextureCoordinate() {coordinateIndex=1}, Time=Constant() {r=12})) {r=true}
  n12 = Multiply(A=Sine(None=Multiply(A=@n11) {constB=6})) {constB=0.5}
  n13 = SmoothStep(Value=Saturate(None=Add(A=@n12) {constB=0.5})) {constMin=0.7}
  MaterialFunctionCall(Emissive=Multiply(A=LinearInterpolate(A=@n9, B=@n10, Alpha=@n13), B=@n1.RGB)) {materialFunction=&MF_FX_EmissiveAdjustments}
 ]
}
