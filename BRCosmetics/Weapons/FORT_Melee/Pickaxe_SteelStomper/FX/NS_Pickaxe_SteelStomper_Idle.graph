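# Idle Niagara effect for the Steel Stomper pickaxe that sheds orange saw sparks from a skeletal mesh bone or socket.
# Vortex forces swirl circular saw streaks, while straight and round sparks get downward velocity, drag and gravity.
# Circular and straight sparks have GPU emitters plus lower-rate CPU "MedLow" versions; round sparks are CPU-only.
# Spawn rates follow a system curve and drop to zero while swinging or in the front end, with time-of-day color tinting.
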NiagaraSystem /BRCosmetics/Weapons/FORT_Melee/Pickaxe_SteelStomper/FX/NS_Pickaxe_SteelStomper_Idle
dependencies {
 usedDataInterfaces [&NiagaraDataInterfaceCurve, &NiagaraDataInterfaceSkeletalMesh, &NiagaraDataInterfaceArrayFloat]
 usedDynamicInputs [&Boolean_NotOperation]
 usedModules [
  &self:SystemSpawnScript.SystemScriptSource.SystemScriptGraph.NiagaraNodeAssignment_0.SetVariables_#1
  &SystemState
  &DoOnce
  &Timeline
  &self:SystemSpawnScript.SystemScriptSource.SystemScriptGraph.NiagaraNodeAssignment_1.SetVariables_#2
 ]
}
emitters {
 RoundSparks {
  emitterUpdateScript [
   &EmitterState()
   &SpawnRate(
    Spawn Probability=0.7
    SpawnRate=&MakeCustomFloatFromBool(
     Bool=$User.bIsSwinging
     True Float=0
     False Float=&FN_FrontEndFloatModifier(Gameplay=&Multiply_Float(A=$System.SpawnRateCurve, B=40), Front End=0)
    )
    Use Spawn Probability=true
   )
  ]
  particleSpawnScript [
   &InitializeParticle(
    Color=&MultiplyLinearColor(
     Linear Color=&RandomRangeLinearColor(Minimum=<rgba 5 2.8079 0.9465 1>, Maximum=<rgba 5 0.397 0 1>)
     Scale Factor=&MakeVector4_FromLinearColor(LinearColor=&FNTimeOfDayColor(
       Night Color=<rgba 0.4143 0.3837 0.6615 1>
       Morning Color=<rgba 0.901 0.5732 0.6456 1>
       Day Color=<rgba 1 1 1 1>
       Evening Color=<rgba 1 1 1 1>
      ))
    )
    Lifetime Max=0.2
    Lifetime Min=0.12
    Lifetime Mode=::NewEnumerator1
    Position Offset=<-60 0 0>
    Sprite Size Mode=::NewEnumerator1
    Uniform Sprite Size Max=15
    Uniform Sprite Size Min=8
    UsePositionOffset=true
    Write Material Random=true
   )
   &SkeletalMeshLocation(
    Bone / Socket Index=6
    MaskSpawnGroup=false
    Mesh Sampling Type=::NewEnumerator5
    Sampled Position Offset=<0 12 10>
    Skeletal Mesh=$System.SkeletalMesh
    Socket Sampling Mode=::NewEnumerator5
   )
   &ShapeLocation(
    Cylinder Distribution=::NewEnumerator2
    Cylinder Height=2
    Cylinder Radius=15
    MaskSpawnGroup=false
    Rotation Mode=::NewEnumerator2
    Shape Origin=$Particles.SkeletalMeshLocation.SampledPosition
    Shape Primitive=::NewEnumerator1
    U Position=&RandomRangeFloat(Minimum=0, Maximum=0.34)
    UseOwnerScale=true
    V Position=&RandomRangeFloat(Minimum=0, Maximum=1)
    Yaw / Pitch / Roll=<0 0.25 0>
   ) {enabled=false}
   &AddVelocity(Velocity=&RandomRangeVector(Minimum=<-60 -100 -350>, Maximum=<60 100 -800>))
   &InheritVelocity(Inherited Velocity Amount Scale=<1 1 0.45>, Inherited Velocity Speed Limit=150, Speed Limit=true, UseSpeedThreshold=false) {enabled=false}
  ]
  particleUpdateScript [
   &Drag(Drag=10, Use Linear Drag=true, Use Rotational Drag=true)
   &GravityForce()
   &SolveForcesAndVelocity(Clamp Velocity=false, Limit Acceleration=false)
   &ParticleState()
   &ScaleSpriteSizeBySpeed(Max Scale Factor=<1.1 1>, Min Scale Factor=<0.7 1>)
   &ScaleColor(
    Scale RGB=&VectorFromFloat(Value=&FloatFromCurve(FloatCurve=DataInterface() {propertyValues {Curve {keys [<at 0 1>,<at 0.4 0.15>,<at 1 0>]}}}))
    ScaleA=true
    ScaleRGB=true
   )
  ]
  rendererClasses [&NiagaraSpriteRendererProperties]
  renderers [{rendererClass=&NiagaraSpriteRendererProperties}]
 }
 SawSparksCircular {
  simTarget=GPUComputeSim
  emitterUpdateScript [
   &EmitterState()
   &SpawnRate(
    SpawnRate=&MakeCustomFloatFromBool(
     Bool=$User.bIsSwinging
     True Float=0
     False Float=&FN_FrontEndFloatModifier(Gameplay=&Multiply_Float(A=$System.SpawnRateCurve, B=250), Front End=0)
    )
    Use Spawn Probability=false
   )
  ]
  particleSpawnScript [
   &InitializeParticle(
    Color=&MultiplyLinearColor(
     Linear Color=&RandomRangeLinearColor(Minimum=<rgba 5 2.8079 0.9465 1>, Maximum=<rgba 5 0.397 0 1>)
     Scale Factor=&MakeVector4_FromLinearColor(LinearColor=&FNTimeOfDayColor(
       Night Color=<rgba 0.4143 0.3837 0.6615 1>
       Morning Color=<rgba 0.901 0.5732 0.6456 1>
       Day Color=<rgba 1 1 1 1>
       Evening Color=<rgba 1 1 1 1>
      ))
    )
    Lifetime Max=0.2
    Lifetime Min=0.12
    Lifetime Mode=::NewEnumerator1
    Position Offset=<-60 0 0>
    Sprite Size Max=<0.8 7>
    Sprite Size Min=<0.4 2.5>
    Sprite Size Mode=::NewEnumerator2
    UsePositionOffset=true
    Write Material Random=true
   )
   &SkeletalMeshLocation(
    Bone / Socket Index=6
    MaskSpawnGroup=false
    Mesh Sampling Type=::NewEnumerator5
    Skeletal Mesh=$System.SkeletalMesh
    Socket Sampling Mode=::NewEnumerator5
   )
   &ShapeLocation(
    Cylinder Distribution=::NewEnumerator2
    Cylinder Height=2
    Cylinder Radius=12
    MaskSpawnGroup=false
    Rotation Mode=::NewEnumerator2
    Shape Origin=$Particles.SkeletalMeshLocation.SampledPosition
    Shape Primitive=::NewEnumerator1
    U Position=&RandomRangeFloat(Minimum=0, Maximum=0.35)
    UseOwnerScale=true
    V Position=&RandomRangeFloat(Minimum=0, Maximum=1)
    Yaw / Pitch / Roll=<0 0.25 0>
   )
   &InheritVelocity(Inherited Velocity Amount Scale=<1 1 0.45>, Inherited Velocity Speed Limit=150, Speed Limit=true, UseSpeedThreshold=false) {enabled=false}
  ]
  particleUpdateScript [
   &VortexForce(
    UseOriginOffset=false
    Vortex Axis=&RandomRangeVector(Minimum=<1 -0.25 0>, Maximum=<1 0.25 0>)
    Vortex Axis Coordinate Space:ENiagaraCoordinateSpace=::Local
    Vortex Force Amount=&RandomRangeFloat(Minimum=-1000, Maximum=-2000)
    Vortex Origin:NiagaraPosition=$Particles.SkeletalMeshLocation.SampledPosition
   )
   &Drag(Drag=4, Use Linear Drag=true, Use Rotational Drag=true)
   &SolveForcesAndVelocity(Clamp Velocity=false, Limit Acceleration=false)
   &ParticleState()
   &ScaleSpriteSizeBySpeed(Max Scale Factor=<1.1 10>, Min Scale Factor=<0.7 1.2>)
   /SetVariables(
    Particles.GlowColor=&RandomRangeLinearColor(Minimum=<rgba 3 0.7185 0 0.3>, Maximum=<rgba 3 2.2678 0 0.3>)
    Particles.GlowSize=&RandomRangeVector2D(Minimum=<1 10>, Maximum=<2 20>)
   ) {isSetParameters=true, module=SetVariables_#3}
  ]
  rendererClasses [&NiagaraSpriteRendererProperties]
  renderers [
   {rendererClass=&NiagaraSpriteRendererProperties}
   {rendererClass=&NiagaraSpriteRendererProperties, rendererIndex=1}
  ]
 }
 SawSparksMedLow {
  emitterUpdateScript [
   &EmitterState()
   &SpawnRate(
    SpawnRate=&MakeCustomFloatFromBool(
     Bool=$User.bIsSwinging
     True Float=0
     False Float=&FN_FrontEndFloatModifier(Gameplay=&Multiply_Float(A=$System.SpawnRateCurve, B=66), Front End=0)
    )
    Use Spawn Probability=false
   )
  ]
  particleSpawnScript [
   &InitializeParticle(
    Color=&MultiplyLinearColor(
     Linear Color=&RandomRangeLinearColor(Minimum=<rgba 5 2.8079 0.9465 1>, Maximum=<rgba 5 0.397 0 1>)
     Scale Factor=&MakeVector4_FromLinearColor(LinearColor=&FNTimeOfDayColor(
       Night Color=<rgba 0.4143 0.3837 0.6615 1>
       Morning Color=<rgba 0.901 0.5732 0.6456 1>
       Day Color=<rgba 1 1 1 1>
       Evening Color=<rgba 1 1 1 1>
      ))
    )
    Lifetime Max=0.2
    Lifetime Min=0.12
    Lifetime Mode=::NewEnumerator1
    Position Offset=<-60 0 0>
    Sprite Size Max=<0.8 7>
    Sprite Size Min=<0.4 2.5>
    Sprite Size Mode=::NewEnumerator2
    UsePositionOffset=true
    Write Material Random=true
   )
   &SkeletalMeshLocation(
    Bone / Socket Index=6
    MaskSpawnGroup=false
    Mesh Sampling Type=::NewEnumerator5
    Skeletal Mesh=$System.SkeletalMesh
    Socket Sampling Mode=::NewEnumerator5
   )
   &ShapeLocation(
    Cylinder Distribution=::NewEnumerator2
    Cylinder Height=2
    Cylinder Radius=12
    MaskSpawnGroup=false
    Rotation Mode=::NewEnumerator2
    Shape Origin=$Particles.SkeletalMeshLocation.SampledPosition
    Shape Primitive=::NewEnumerator1
    U Position=&RandomRangeFloat(Minimum=0, Maximum=0.35)
    UseOwnerScale=true
    V Position=&RandomRangeFloat(Minimum=0, Maximum=1)
    Yaw / Pitch / Roll=<0 0.25 0>
   )
   &InheritVelocity(Inherited Velocity Amount Scale=<1 1 0.45>, Inherited Velocity Speed Limit=150, Speed Limit=true, UseSpeedThreshold=false) {enabled=false}
  ]
  particleUpdateScript [
   &VortexForce(
    UseOriginOffset=false
    Vortex Axis=&RandomRangeVector(Minimum=<1 -0.25 0>, Maximum=<1 0.25 0>)
    Vortex Axis Coordinate Space:ENiagaraCoordinateSpace=::Local
    Vortex Force Amount=&RandomRangeFloat(Minimum=-1000, Maximum=-2000)
    Vortex Origin:NiagaraPosition=$Particles.SkeletalMeshLocation.SampledPosition
   )
   &Drag(Drag=4, Use Linear Drag=true, Use Rotational Drag=true)
   &SolveForcesAndVelocity(Clamp Velocity=false, Limit Acceleration=false)
   &ParticleState()
   &ScaleSpriteSizeBySpeed(Max Scale Factor=<1.1 10>, Min Scale Factor=<0.7 1.2>)
   /SetVariables(
    Particles.GlowColor=&RandomRangeLinearColor(Minimum=<rgba 3 0.7185 0 0.3>, Maximum=<rgba 3 2.2678 0 0.3>)
    Particles.GlowSize=&RandomRangeVector2D(Minimum=<1 10>, Maximum=<2 20>)
   ) {isSetParameters=true, module=SetVariables_#3}
  ]
  rendererClasses [&NiagaraSpriteRendererProperties]
  renderers [
   {rendererClass=&NiagaraSpriteRendererProperties}
   {rendererClass=&NiagaraSpriteRendererProperties, rendererIndex=1}
  ]
 }
 StraightSparks {
  simTarget=GPUComputeSim
  emitterUpdateScript [
   &EmitterState()
   &SpawnRate(
    SpawnRate=&MakeCustomFloatFromBool(
     Bool=$User.bIsSwinging
     True Float=0
     False Float=&FN_FrontEndFloatModifier(Gameplay=&Multiply_Float(A=$System.SpawnRateCurve, B=100), Front End=0)
    )
    Use Spawn Probability=false
   )
  ]
  particleSpawnScript [
   &InitializeParticle(
    Color=&MultiplyLinearColor(
     Linear Color=&RandomRangeLinearColor(Minimum=<rgba 5 2.8079 0.9465 1>, Maximum=<rgba 5 0.397 0 1>)
     Scale Factor=&MakeVector4_FromLinearColor(LinearColor=&FNTimeOfDayColor(
       Night Color=<rgba 0.4143 0.3837 0.6615 1>
       Morning Color=<rgba 0.901 0.5732 0.6456 1>
       Day Color=<rgba 1 1 1 1>
       Evening Color=<rgba 1 1 1 1>
      ))
    )
    Lifetime Max=0.2
    Lifetime Min=0.12
    Lifetime Mode=::NewEnumerator1
    Position Offset=<-60 0 0>
    Sprite Size Max=<0.8 7>
    Sprite Size Min=<0.4 2.5>
    Sprite Size Mode=::NewEnumerator2
    UsePositionOffset=true
    Write Material Random=true
   )
   &SkeletalMeshLocation(
    Bone / Socket Index=6
    MaskSpawnGroup=false
    Mesh Sampling Type=::NewEnumerator5
    Sampled Position Offset=<0 12 10>
    Skeletal Mesh=$System.SkeletalMesh
    Socket Sampling Mode=::NewEnumerator5
   )
   &ShapeLocation(
    Cylinder Distribution=::NewEnumerator2
    Cylinder Height=2
    Cylinder Radius=15
    MaskSpawnGroup=false
    Rotation Mode=::NewEnumerator2
    Shape Origin=$Particles.SkeletalMeshLocation.SampledPosition
    Shape Primitive=::NewEnumerator1
    U Position=&RandomRangeFloat(Minimum=0, Maximum=0.34)
    UseOwnerScale=true
    V Position=&RandomRangeFloat(Minimum=0, Maximum=1)
    Yaw / Pitch / Roll=<0 0.25 0>
   ) {enabled=false}
   &AddVelocity(Velocity=&RandomRangeVector(Minimum=<-60 -100 -250>, Maximum=<60 100 -600>))
   &InheritVelocity(Inherited Velocity Amount Scale=<1 1 0.45>, Inherited Velocity Speed Limit=150, Speed Limit=true, UseSpeedThreshold=false) {enabled=false}
  ]
  particleUpdateScript [
   &Drag(Drag=10, Use Linear Drag=true, Use Rotational Drag=true)
   &GravityForce()
   &SolveForcesAndVelocity(Clamp Velocity=false, Limit Acceleration=false)
   &ParticleState()
   &ScaleSpriteSizeBySpeed(Max Scale Factor=<1.1 10>, Min Scale Factor=<0.7 1.2>)
   /SetVariables(
    Particles.GlowColor=&RandomRangeLinearColor(Minimum=<rgba 2 0.3356 0 0.3>, Maximum=<rgba 3 2.1235 0.3392 0.3>)
    Particles.GlowSize=&RandomRangeVector2D(Minimum=<1 10>, Maximum=<2 20>)
   ) {isSetParameters=true, module=SetVariables_#3}
   &ScaleColor(
    Scale RGB=&VectorFromFloat(Value=&FloatFromCurve(FloatCurve=DataInterface() {propertyValues {Curve {keys [<at 0 1>,<at 0.4 0.15>,<at 1 0>]}}}))
    ScaleA=true
    ScaleRGB=true
   )
   /SetVariables(Particles.Vector 2D:Vector2f=<10 10>) {isSetParameters=true, module=SetVariables_#4}
  ]
  rendererClasses [&NiagaraSpriteRendererProperties]
  renderers [
   {rendererClass=&NiagaraSpriteRendererProperties}
   {rendererClass=&NiagaraSpriteRendererProperties, rendererIndex=1}
  ]
 }
 StraightSparksMedLow {
  emitterUpdateScript [
   &EmitterState()
   &SpawnRate(
    SpawnRate=&MakeCustomFloatFromBool(
     Bool=$User.bIsSwinging
     True Float=0
     False Float=&FN_FrontEndFloatModifier(Gameplay=&Multiply_Float(A=$System.SpawnRateCurve, B=60), Front End=0)
    )
    Use Spawn Probability=false
   )
  ]
  particleSpawnScript [
   &InitializeParticle(
    Color=&MultiplyLinearColor(
     Linear Color=&RandomRangeLinearColor(Minimum=<rgba 5 2.8079 0.9465 1>, Maximum=<rgba 5 0.397 0 1>)
     Scale Factor=&MakeVector4_FromLinearColor(LinearColor=&FNTimeOfDayColor(
       Night Color=<rgba 0.4143 0.3837 0.6615 1>
       Morning Color=<rgba 0.901 0.5732 0.6456 1>
       Day Color=<rgba 1 1 1 1>
       Evening Color=<rgba 1 1 1 1>
      ))
    )
    Lifetime Max=0.2
    Lifetime Min=0.12
    Lifetime Mode=::NewEnumerator1
    Position Offset=<-60 0 0>
    Sprite Size Max=<0.8 7>
    Sprite Size Min=<0.4 2.5>
    Sprite Size Mode=::NewEnumerator2
    UsePositionOffset=true
    Write Material Random=true
   )
   &SkeletalMeshLocation(
    Bone / Socket Index=6
    MaskSpawnGroup=false
    Mesh Sampling Type=::NewEnumerator5
    Sampled Position Offset=<0 12 10>
    Skeletal Mesh=$System.SkeletalMesh
    Socket Sampling Mode=::NewEnumerator5
   )
   &ShapeLocation(
    Cylinder Distribution=::NewEnumerator2
    Cylinder Height=2
    Cylinder Radius=15
    MaskSpawnGroup=false
    Rotation Mode=::NewEnumerator2
    Shape Origin=$Particles.SkeletalMeshLocation.SampledPosition
    Shape Primitive=::NewEnumerator1
    U Position=&RandomRangeFloat(Minimum=0, Maximum=0.34)
    UseOwnerScale=true
    V Position=&RandomRangeFloat(Minimum=0, Maximum=1)
    Yaw / Pitch / Roll=<0 0.25 0>
   ) {enabled=false}
   &AddVelocity(Velocity=&RandomRangeVector(Minimum=<-60 -100 -250>, Maximum=<60 100 -600>))
   &InheritVelocity(Inherited Velocity Amount Scale=<1 1 0.45>, Inherited Velocity Speed Limit=150, Speed Limit=true, UseSpeedThreshold=false) {enabled=false}
  ]
  particleUpdateScript [
   &Drag(Drag=10, Use Linear Drag=true, Use Rotational Drag=true)
   &GravityForce()
   &SolveForcesAndVelocity(Clamp Velocity=false, Limit Acceleration=false)
   &ParticleState()
   &ScaleSpriteSizeBySpeed(Max Scale Factor=<1.1 10>, Min Scale Factor=<0.7 1.2>)
   /SetVariables(
    Particles.GlowColor=&RandomRangeLinearColor(Minimum=<rgba 2 0.3356 0 0.3>, Maximum=<rgba 3 2.1235 0.3392 0.3>)
    Particles.GlowSize=&RandomRangeVector2D(Minimum=<1 10>, Maximum=<2 20>)
   ) {isSetParameters=true, module=SetVariables_#3}
   &ScaleColor(
    Scale RGB=&VectorFromFloat(Value=&FloatFromCurve(FloatCurve=DataInterface() {propertyValues {Curve {keys [<at 0 1>,<at 0.4 0.15>,<at 1 0>]}}}))
    ScaleA=true
    ScaleRGB=true
   )
   /SetVariables(Particles.Vector 2D:Vector2f=<10 10>) {isSetParameters=true, module=SetVariables_#4}
  ]
  rendererClasses [&NiagaraSpriteRendererProperties]
  renderers [
   {rendererClass=&NiagaraSpriteRendererProperties}
   {rendererClass=&NiagaraSpriteRendererProperties, rendererIndex=1}
  ]
 }
}
summary {
 systemName=NS_Pickaxe_SteelStomper_Idle
 emitters [
  {emitterName=SawSparksCircular, simTarget=GPUComputeSim, rendererClasses [&NiagaraSpriteRendererProperties]}
  {emitterName=StraightSparks, simTarget=GPUComputeSim, rendererClasses [&NiagaraSpriteRendererProperties]}
  {emitterName=RoundSparks, rendererClasses [&NiagaraSpriteRendererProperties]}
  {emitterName=SawSparksMedLow, rendererClasses [&NiagaraSpriteRendererProperties]}
  {emitterName=StraightSparksMedLow, rendererClasses [&NiagaraSpriteRendererProperties]}
 ]
 userVariables [
  User.bIsFrontEnd {type=NiagaraBool}
  User.bIsFrontEndPreview {type=NiagaraBool}
  User.bIsSwinging {type=NiagaraBool}
 ]
}
