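# Niagara trail effect for the Steel Stomper pickaxe, built from two ribbon emitters and one sprite emitter that read the weapon's skeletal mesh.
# FN_AnimTrail draws three socket-pair trails with per-spawn-group color and width, while FN_AnimTrail001 adds another translucent trail, all fading quickly.
# The "Empty" emitter spawns short-lived orange sprites around a socket, inheriting velocity and scaling with speed.
# Time-of-day inputs tint the sprites and some trails.
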NiagaraSystem /BRCosmetics/Weapons/FORT_Melee/Pickaxe_SteelStomper/FX/NS_Pickaxe_SteelStomper_Trail
dependencies {
 usedDataInterfaces [&NiagaraDataInterfaceCurve, &NiagaraDataInterfaceSkeletalMesh]
 usedDynamicInputs []
 usedModules [
  &self:SystemSpawnScript.SystemScriptSource.SystemScriptGraph.NiagaraNodeAssignment_0.SetVariables_#1
  &SystemState
 ]
}
emitters {
 Empty {
  emitterUpdateScript [
   &EmitterState()
   &SpawnRate(SpawnRate=60, Use Spawn Probability=false)
  ]
  particleSpawnScript [
   &InitializeParticle(
    Color=&MultiplyLinearColor(
     Linear Color=<rgba 5 2.8079 0.9465 1>
     Scale Factor=&MakeVector4_FromLinearColor(LinearColor=&FNTimeOfDayColor(Night Color=<rgba 0.5579 0.5166 0.8906 1>, Morning Color=<rgba 1 0.7119 0.7721 1>, Day Color=<rgba 1 1 1 1>, Evening Color=<rgba 1 1 1 1>))
    )
    Lifetime Max=0.25
    Lifetime Min=0.15
    Lifetime Mode=::NewEnumerator1
    Position Offset=<-60 0 0>
    Sprite Size Max=<0.8 8>
    Sprite Size Min=<0.4 2>
    Sprite Size Mode=::NewEnumerator2
    UsePositionOffset=true
    Write Material Random=true
   )
   &SkeletalMeshLocation(
    Bone / Socket Index=6
    MaskSpawnGroup=false
    Mesh Sampling Type=::NewEnumerator5
    Sampled Position Offset=&RandomRangeVector(Minimum=<15 15 15>, Maximum=<-15 -15 -15>)
    Skeletal Mesh=$System.SkeletalMesh
    Socket Sampling Mode=::NewEnumerator5
   )
   &InheritVelocity(Inherited Velocity Amount Scale=<1 1 0.45>, Inherited Velocity Speed Limit=150, Speed Limit=true, UseSpeedThreshold=false)
  ]
  particleUpdateScript [
   &SolveForcesAndVelocity(Clamp Velocity=false, Limit Acceleration=false)
   &ParticleState()
   &ScaleSpriteSizeBySpeed(Max Scale Factor=<1.1 2>, Min Scale Factor=<0.6 1.2>)
   /SetVariables(Particles.GlowColor=<rgba 5 1.8559 0 0.3>, Particles.GlowSize=<2 8>) {isSetParameters=true}
  ]
  rendererClasses [&NiagaraSpriteRendererProperties]
  renderers [
   {rendererClass=&NiagaraSpriteRendererProperties}
   {rendererClass=&NiagaraSpriteRendererProperties, rendererIndex=1}
  ]
 }
 FN_AnimTrail {
  emitterSpawnScript [&FN_RibbonIDOffsetUtility()]
  emitterUpdateScript [
   &EmitterState()
   &FN_SpawnAnimTrail(Number of Anim Trails=3)
  ]
  particleSpawnScript [
   &FN_InitializeAnimTrailBoneIndecies(First Bone or Socket Index=2, Second Bone or Socket Index=3) {module=FN_InitializeAnimTrailBoneIndecies001}
   &FN_InitializeAnimTrailBoneIndecies(AnimTrail Selection=2, First Bone or Socket Index=4, Second Bone or Socket Index=5) {module=FN_InitializeAnimTrailBoneIndecies002}
   &FN_InitializeAnimTrailBoneIndecies(AnimTrail Selection=1)
   &FN_InitializeAnimTrail(
    AnimTrailWidth=&MaskFloatBySpawnGroup(Masked Float=1, Passthrough Float=1.6, Spawn Group Mask=2)
    Color=&MaskLinearColorBySpawnGroup(
     Masked Linear Color=<rgba 0.2135 0.0075 0 2>
     Passthrough Linear Color=&MaskLinearColorBySpawnGroup(
      Masked Linear Color=<rgba 0.0366 0.0415 0.0469 8>
      Passthrough Linear Color=&FNTimeOfDayColor(
       Night Color=<rgba 0.2704 0.238 0.3698 1>
       Morning Color=<rgba 1 0.6951 0.735 0.4>
       Day Color=<rgba 0.3698 0.3408 0.3296 1>
       Evening Color=<rgba 0.3698 0.3408 0.3296 1>
      )
      Spawn Group Mask=1
     )
    )
    Lifetime=0.15
    SkeletalMesh=$System.SkeletalMesh
   )
   &AddVelocity(Velocity=<0 400 0>) {enabled=false}
  ]
  particleUpdateScript [
   &SolveForcesAndVelocity(Clamp Velocity=false, Limit Acceleration=false)
   &FN_UpdateAnimTrail(Locked Particle RibbonWidth=1, SkeletalMesh=$System.SkeletalMesh) {module=FN_UpdateAnimTrail001}
   &ScaleColor(
    Scale Alpha=&FloatFromCurve(FloatCurve=DataInterface() {propertyValues {Curve {keys [{interpMode=RCIM_Cubic, leaveTangent=-3.8387, tangentMode=RCTM_Break, time=0, value=1}, <at 0.6 0>]}}})
    ScaleA=true
    ScaleRGB=true
   )
  ]
  rendererClasses [&NiagaraRibbonRendererProperties]
  renderers [
   {rendererClass=&NiagaraRibbonRendererProperties}
   {rendererClass=&NiagaraRibbonRendererProperties, rendererIndex=1}
  ]
 }
 FN_AnimTrail001 {
  emitterSpawnScript [&FN_RibbonIDOffsetUtility()]
  emitterUpdateScript [
   &EmitterState()
   &FN_SpawnAnimTrail()
  ]
  particleSpawnScript [
   &FN_InitializeAnimTrailBoneIndecies()
   &FN_InitializeAnimTrail(
    AnimTrailWidth=2
    Color=&FNTimeOfDayColor(
     Night Color=<rgba 0.6241 0.4484 1 0.164>
     Morning Color=<rgba 1 0.6556 0.6755 0.3>
     Day Color=<rgba 1 1 1 0.4>
     Evening Color=<rgba 1 1 1 0.4>
    )
    Lifetime=0.15
    SkeletalMesh=$System.SkeletalMesh
   )
   &AddVelocity(Velocity=<0 400 0>) {enabled=false}
  ]
  particleUpdateScript [
   &SolveForcesAndVelocity(Clamp Velocity=false, Limit Acceleration=false)
   &FN_UpdateAnimTrail(Locked Particle RibbonWidth=1, SkeletalMesh=$System.SkeletalMesh) {module=FN_UpdateAnimTrail001}
   &ScaleColor(
    Scale Alpha=&FloatFromCurve(FloatCurve=DataInterface() {propertyValues {Curve {keys [
         {interpMode=RCIM_Cubic, leaveTangent=-3.8387, tangentMode=RCTM_Break, time=0, value=1}
         {arriveTangent=-0.8463, tangentMode=RCTM_User, time=0.5, value=0}
        ]}}})
    ScaleA=true
    ScaleRGB=true
   )
  ]
  rendererClasses [&NiagaraRibbonRendererProperties]
  renderers [
   {rendererClass=&NiagaraRibbonRendererProperties}
   {rendererClass=&NiagaraRibbonRendererProperties, rendererIndex=1}
  ]
 }
}
summary {
 systemName=NS_Pickaxe_SteelStomper_Trail
 emitters [
  {emitterName=FN_AnimTrail, rendererClasses [&NiagaraRibbonRendererProperties]}
  {emitterName=FN_AnimTrail001, rendererClasses [&NiagaraRibbonRendererProperties]}
  {emitterName=Empty, rendererClasses [&NiagaraSpriteRendererProperties]}
 ]
 userVariables [User.RibbonIDOffset {type=NiagaraInt32}]
}
