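# Animation Blueprint for the SteelStomper pickaxe that continuously spins its circular saw blade by additively rolling the blade bone.
# Each update accumulates a wrapped angle from RotationSpeed, which jumps high while a boost flag is set and eases back to idle otherwise.
# Delays toggle that flag after the owner's PlayerMeleeCombo montage starts or stops, outside the front end.
# The pickaxe special action event also triggers a timed boost.
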AnimBlueprint /BRCosmetics/Weapons/FORT_Melee/Pickaxe_SteelStomper/Meshes/SteelStomper_Axe_AnimBP
target_skeleton=&/SteelStomper_Axe_Skeleton
graphs [
 AnimGraph {nodes [
   AnimGraphNode_SubInput_0 = Input Pose()
   n1 = "Transform (Modify) Bone - Bone: dyn_circularSawBlade"!(
    ComponentPose=Local To Component(LocalPose=@AnimGraphNode_SubInput_0)
    Alpha=1
    Translation="0,0,0"
    Rotation_Roll:float=Get!() -> (rotation of saw blade_Roll:float, rotation of saw blade_Pitch:float, rotation of saw blade_Yaw:float)
    Rotation_Pitch:float=0
    Rotation_Yaw:float=0
    Scale="1,1,1"
   ) -> (Pose) {rotationMode=BMM_Additive, rotationSpace=BCS_BoneSpace, alphaBoolBlend {blendOption=Linear}, boneToModify {boneName=dyn_circularSawBlade}}
   Output Pose(Result=Component To Local(ComponentPose=@n1)) {layerGroup=None}
  ]}
 EventGraph {nodes [
   Event BlueprintUpdateAnimation()
   > n1 = Sequence!(execute) -> (then_0, then_1, then_2:Exec, then_3:Exec)
   n2 = GetAnimInstance(self=GetOwnerSkeletalMeshComponent())
   Set is Animation player melee combo(
    execute=@n1.then_3
    is Animation player melee combo:bool="Equal (Object)"(
     A=GetCurrentActiveMontage!(self:Anim Instance Object Reference=@n2) -> (ReturnValue:Anim Montage Object Reference)
     B=/Game/Animation/Game/MainPlayer/Montages/Melee/PlayerMeleeCombo
    ) {is=title K2Node_CallFunction}
   ) -> (Output_Get:bool)
   n3 = Branch(execute=@n1.then_1, Condition=Get animation plus delay() -> (animation plus delay:bool))
   > Set RotationSpeed(RotationSpeed:double=900) -> (Output_Get:double)
   Set RotationSpeed(
    execute=@n3.else
    RotationSpeed=FInterpTo(Current=Get RotationSpeed() -> (RotationSpeed:double), Target=100, DeltaTime=GetWorldDeltaSeconds!() -> (ReturnValue:double), InterpSpeed=1)
   )
   n4 = GetWorldDeltaSeconds()
   n5 = Get RotationSpeed()
   n6 = Set CurrentAngle(
    execute=@n1.then_2
    CurrentAngle:double=float + float(A=Get CurrentAngle() -> (CurrentAngle:double), B=float * float(A=@n4, B=@n5) {is=title K2Node_PromotableOperator})
   ) -> (Output_Get:double)
   > n7 = Set CurrentAngle(CurrentAngle="% (Float)"!(A:double=@n6, B:double=360) -> (ReturnValue:double))
   > Set rotation of saw blade(rotation of saw blade:Rotator=MakeRotator(Roll=@n7)) -> (Output_Get:Rotator)
   n8 = Branch(execute=@n1.then_0, Condition=Get bIsInFrontEnd())
   n9 = Branch(execute=@n8.else, Condition=Get is Animation player melee combo() -> (is Animation player melee combo:bool))
   > Delay(Duration=1)
   > Set animation plus delay(animation plus delay:bool=true) -> (Output_Get:bool)
   Delay(execute=@n9.else, Duration=1)
   > Set animation plus delay(animation plus delay=false)
   Event OnPickaxeSpecialAction!() -> (OutputDelegate, then, ActionName:Name=None) {is=title K2Node_Event}
   > Delay(Duration=0.85)
   > Set animation plus delay(animation plus delay=true)
   > Delay(Duration=3)
   > Set animation plus delay(animation plus delay=false)
   "Clamp (Float)"(Value=0, Min=100, Max=360)
  ]}
]
variables [
 animation plus delay {type=boolean}
 B_ISROTATE {type=boolean}
 CurrentAngle {type=number, default=207.5236}
 is Animation player melee combo {type=boolean}
 Output {type=SkeletalMeshComponentBudgeted}
 rotation of saw blade {type=Rotator, default=<rot 0 0 207.5236>}
 RotationSpeed {type=number, default=100}
]
